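; Player controller component that auto-pings a nearby vending machine when the player's gold reaches a threshold.
; BeginPlay caches the marker component and loads or creates a save object holding the ping count.
; Inventory updates pass through a delta gate, so only threshold crossings trigger a radius search for the closest machine inside the safe zone.
; A hotfixable toggle and saved maximum count limit pings, which are marked on the server.
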(event EventBeginPlay
  (|CallParentFunction)
  (bind _asfort_player_controller (Utilities|Casting|CastToFortPlayerController (Components|GetOwner)))
  (bind _output_get (Variables|Default|SetPlayerController _asfort_player_controller))
  (Variables|Default|SetMarkerComponent (Actor|GetComponentbyClass _output_get "/Script/FortniteGame.AthenaMarkerComponent"))
  ; Save data
  (bind _returnvalue (SaveGame|LoadGamefromSlot (Variables|Default|GetVending%20Machine%20Save%20Slot%20Name)))
  (bind _asbp_vending_machine_save (Utilities|Casting|CastToBP_VendingMachine_Save _returnvalue)
    (:then
      (Variables|Default|SetVending%20Machine%20Save _asbp_vending_machine_save))
    (:CastFailed
      (bind _returnvalue_2 (SaveGame|CreateSaveGameObject "/VendingMachine/ActorBlueprints/BP_VendingMachine_Save.BP_VendingMachine_Save_C"))
      (Variables|Default|SetVending%20Machine%20Save _returnvalue_2))))

(fn OnInventoryItemsUpdated
  (params (UpdatedItems :type Array%20of%20Fort%20Item%20Object%20References))
  (bind _playercontroller (Variables|Default|GetPlayerController))
  (bind _returnvalue (Game|Player|GetPlayerPawn _playercontroller))
  (bind _vending_machine_save (Variables|Default|GetVending%20Machine%20Save))
  (if (and (Utilities|IsValid (Variables|Default|GetMarkerComponent) :!class /Script/Engine.KismetSystemLibrary) (Utilities|IsValid (Variables|Default|GetPlayerController) :!class /Script/Engine.KismetSystemLibrary))
    ;[1 0.2918 0.6795 0.502] Hotfixable variable to enable or disable Gold Threshold related Vending Machine auto pings
    (if (ScalableFloat|GetValueasBool (Variables|Default|GetGoldPingEnabled) 1.0)
      (bind _returnvalue_2 (InventoryManagement|GetItemCountforItemDefinition (Variables|Default|GetPlayerController) (Variables|Properties|GetGoldItemDefinition)))
      (bind _updated_variable (Utilities|FlowControl|DeltaGateWildcard (Variables|Default|GetShouldPingVendingMachine) (>= _returnvalue_2 (ScalableFloat|GetValueasInteger (Variables|Properties|GetGoldThresholdToPing) 1.0)))
        (:Success
          (if _updated_variable
            ; Only show for the first MaxCount pings
            (if (<= (Variables|/VendingMachine/ActorBlueprints/BP_VendingMachine_Save.BP_VendingMachine_Save_C|GetGold%20Ping%20Count (Variables|Default|GetVending%20Machine%20Save)) (ScalableFloat|GetValueatLevel (Variables|Default|GetGoldPingMaxCount) 1.0))
              (Utilities|IsValid _returnvalue
                (:Is%20Valid
                  (bind _output_get (LocalVariables|SetPlayerLocation (Transformation|GetActorLocation _returnvalue)))
                  (bind _outactors (Fortnite|GetAllActorsOfClassinRange (Variables|Properties|GetVendingMachineBaseActorClass) _output_get (ScalableFloat|GetValueatLevel (Variables|Properties|GetPingRadius) 1.0)))
                  (for _array_element _outactors
                    (bind _asbp_vending_machine_base (Utilities|Casting|CastToBP_VendingMachine_Base _array_element))
                    (bind _returnvalue_3 (SafeZone|IsLocationinSafeZone (Transformation|GetActorLocation _asbp_vending_machine_base)))
                    (if _returnvalue_3
                      (Utilities|IsValid (LocalVariables|GetClosestVendingMachine)
                        (:Is%20Valid
                          (if (< (Math|Vector|DistanceSquared(Vector) (Transformation|GetActorLocation _asbp_vending_machine_base) (LocalVariables|GetPlayerLocation)) (Math|Vector|DistanceSquared(Vector) (LocalVariables|GetPlayerLocation) (Transformation|GetActorLocation (LocalVariables|GetClosestVendingMachine))))
                            _))
                        (:Is%20Not%20Valid
                          _))
                      (LocalVariables|SetClosestVendingMachine _asbp_vending_machine_base)))
                  (MarkerComponent|MarkActoronServer (Variables|Default|GetMarkerComponent) (LocalVariables|GetClosestVendingMachine) false true)
                  (Class|BPVendingMachineSave|IncreaseGoldPingCount _vending_machine_save)
                  (SaveGame|SaveGametoSlot _vending_machine_save (Variables|Default|GetVending%20Machine%20Save%20Slot%20Name)))))))))))
